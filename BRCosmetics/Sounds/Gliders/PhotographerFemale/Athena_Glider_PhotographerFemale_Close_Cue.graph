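# Sound cue for closing the PhotographerFemale glider cosmetic in Athena.
# A mixer layers the main close sound with a slightly quieter vocal, picked at equal weight from three takes and varied through a modulator.
# It uses the shared player glider open/close attenuation and concurrency, the glider close property sheet, and the glider landing sound class.
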SoundCue /BRCosmetics/Sounds/Gliders/PhotographerFemale/Athena_Glider_PhotographerFemale_Close_Cue
root=Mixer_0
nodes [
 {id=Mixer_0, children [WavePlayer_1,Modulator_0], properties {inputVolume [1,0.85]}}
 {id=Modulator_0, children [Random_0], properties {pitchMax=1.1}}
 {id=Random_0, children [WavePlayer_0,WavePlayer_2,WavePlayer_3], properties {weights [1,1,1]}}
 {
  id=WavePlayer_0
  properties {soundWaveAssetPtr=&/Athena_Glider_PhotographerFemale_Close_Vocal_01}
 }
 {id=WavePlayer_1, properties {soundWaveAssetPtr=&/Athena_Glider_PhotographerFemale_Close}}
 {
  id=WavePlayer_2
  properties {soundWaveAssetPtr=&/Athena_Glider_PhotographerFemale_Close_Vocal_02}
 }
 {
  id=WavePlayer_3
  properties {soundWaveAssetPtr=&/Athena_Glider_PhotographerFemale_Close_Vocal_03}
 }
]
properties {
 attenuationSettings=/Game/Sounds/Attenuation/Fort_Player_Glider_OpenClose_Athena_Att
 audioPropertiesSheet=/BRCosmetics/Sounds/Settings/PropertySheets/Glider/APS_Cosmetic_Glider_Close
 maxDistance=10000
 soundClassObject=&Fort_Gliders_Land
 volumeMultiplier=1
 concurrencySet [&Athena_GliderOpenClose_Conc]
}
