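# Sound cue played when the PhotographerFemale cosmetic glider opens.
# A mixer layers the base open sound with a quieter vocal branch.
# The vocal branch picks one of four equally weighted vocal waves at random, starts it after a fixed short delay and passes it through a modulator.
# Glider open/close attenuation and concurrency, a glider-open property sheet and a glider sound class govern playback.
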SoundCue /BRCosmetics/Sounds/Gliders/PhotographerFemale/Athena_Glider_PhotographerFemale_Open_Cue
root=Mixer_0
nodes [
 {id=Delay_0, children [Random_0], properties {delayMax=0.5, delayMin=0.5}}
 {id=Mixer_0, children [WavePlayer_1,Modulator_0], properties {inputVolume [1,0.65]}}
 {id=Modulator_0, children [Delay_0]}
 {id=Random_0, children [WavePlayer_0,WavePlayer_2,WavePlayer_3,WavePlayer_4], properties {weights [1,1,1,1]}}
 {
  id=WavePlayer_0
  properties {soundWaveAssetPtr=&/Athena_Glider_PhotographerFemale_Open_Vocal_01}
 }
 {id=WavePlayer_1, properties {soundWaveAssetPtr=&/Athena_Glider_PhotographerFemale_Open}}
 {
  id=WavePlayer_2
  properties {soundWaveAssetPtr=&/Athena_Glider_PhotographerFemale_Open_Vocal_02}
 }
 {
  id=WavePlayer_3
  properties {soundWaveAssetPtr=&/Athena_Glider_PhotographerFemale_Open_Vocal_03}
 }
 {
  id=WavePlayer_4
  properties {soundWaveAssetPtr=&/Athena_Glider_PhotographerFemale_Open_Vocal_04}
 }
]
properties {
 attenuationSettings=/Game/Sounds/Attenuation/Fort_Player_Glider_OpenClose_Athena_Att
 audioPropertiesSheet=/BRCosmetics/Sounds/Settings/PropertySheets/Glider/APS_Cosmetic_Glider_Open
 maxDistance=10000
 soundClassObject=&Fort_Gliders_Land
 volumeMultiplier=1
 concurrencySet [&Athena_GliderOpenClose_Conc]
}
